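# Sound cue for picking up the Birthday variant of the S31 Takedown Medallion.
# A Mixer layers the default S28 medallion pickup SFX with the Birthday stinger, which runs through a Delay node left at its usual settings and plays slightly quieter.
# The cue uses the loot-collect sound class and overrides concurrency to allow only one instance at a time.
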SoundCue /FlexLegendMedallionGameplay/SFX/S31_TakedownMedallions_Pickup_Birthday_Cue
root=Mixer_0
nodes [
 {id=Delay_1, children [WavePlayer_0]}
 {id=Mixer_0, children [WavePlayer_1,Delay_1], properties {inputVolume [1,0.85]}}
 {id=WavePlayer_0, properties {soundWaveAssetPtr=&/S31_Stinger_Medallion_Birthday}}
 {id=WavePlayer_1, properties {soundWaveAssetPtr=/TakedownMedallion/Audio/Medallions/S28/Default_TakedownMedallions_Pickup_SFX}}
]
properties {
 bOverrideConcurrency=true
 maxDistance=2097152
 soundClassObject=&Fort_Loot_Collect
 volumeMultiplier=0.7692
 concurrencyOverrides {maxCount=1}
}
